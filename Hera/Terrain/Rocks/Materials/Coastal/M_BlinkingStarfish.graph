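# Nanite-ready material for a coastal starfish that blinks by swapping between open-eye and closed-eye diffuse textures.
# Time offset by actor position feeds layered sine, cosine and fmod waves, stepped into an irregular per-actor Blink Mask.
# An IsNanite? switch selects texture sets, then flattened normals and adjusted roughness feed MF_FN_DestructibleDamage.
# A breathing offset using BreathIntensity custom primitive data is built, but World Position Offset shows no connection.
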Material /Hera/Terrain/Rocks/Materials/Coastal/M_BlinkingStarfish
graph {
 properties {
  bAutomaticallySetUsageInEditor=false
  bUseMaterialAttributes=true
  bUsedWithInstancedStaticMeshes=true
  bUsedWithNanite=true
  usedShadingModels=Default Lit
 }
 nodes [
  n1 = TextureSample() {samplerType=SAMPLERTYPE_Masks, texture=/Hera/Terrain/Rocks/Textures/Addon/T_Hera_SeaMist_RockClutter_A_B_S_NITE}
  n2 = SetMaterialAttributes!(
   MaterialAttributes=SetMaterialAttributes!(
    MaterialAttributes
    Normal=TextureSample() {samplerType=SAMPLERTYPE_Normal, texture=/Hera/Terrain/Rocks/Textures/Addon/T_Hera_SeaMist_RockClutter_A_B_N_NITE}.RGB
   ) {attributeSetTypes [#1]}
   Roughness=@n1.B
   Metallic=@n1.B
   Specular=@n1.B
  ) {attributeSetTypes [#2, #3, #4]}
  n3 = TextureSample() {texture=/Hera/Terrain/Rocks/Textures/Addon/T_Hera_SeaMist_RockClutter_A_B_D_NITE}
  n4 = TextureSample() {samplerType=SAMPLERTYPE_Masks, texture=/Hera/Terrain/Rocks/Textures/Addon/T_Hera_SeaMist_RockClutter_A_B_S}
  n5 = SetMaterialAttributes!(
   MaterialAttributes=SetMaterialAttributes!(
    MaterialAttributes
    Normal=TextureSample() {samplerType=SAMPLERTYPE_Normal, texture=/Hera/Terrain/Rocks/Textures/Addon/T_Hera_SeaMist_RockClutter_A_B_N}.RGB
   ) {attributeSetTypes [#1]}
   Roughness=@n4.B
   Metallic=@n4.B
   Specular=@n4.B
  ) {attributeSetTypes [#2, #3, #4]}
  n6 = TextureSample() {texture=/Hera/Terrain/Rocks/Textures/Addon/T_Hera_SeaMist_RockClutter_A_B_D}
  BlendMaterialAttributes(
   B=StaticSwitchParameter(
    True=SetMaterialAttributes!(MaterialAttributes=@n2, Base Color=@n3.RGB) {attributeSetTypes [#5]}
    False=SetMaterialAttributes(MaterialAttributes=@n5, Base Color=@n6.RGB) {attributeSetTypes [#5]}
   ) {defaultValue=true, parameterName=IsNanite?}
   Alpha=NamedRerouteUsage()
  )
  Constant() {r=1}
  Constant() {r=5}
  Constant() {r=2}
  n7 = TextureSampleParameter2D() {parameterName=N_NITE, samplerType=SAMPLERTYPE_Normal, texture=/Hera/Terrain/Rocks/Textures/Addon/T_Hera_SeaMist_RockClutter_A_A_N_NITE}
  n8 = TextureSampleParameter2D() {parameterName=S_NITE, samplerType=SAMPLERTYPE_Masks, texture=/Hera/Terrain/Rocks/Textures/Addon/T_Hera_SeaMist_RockClutter_A_A_S_NITE}
  n9 = SetMaterialAttributes!(
   MaterialAttributes=SetMaterialAttributes!(MaterialAttributes, Normal=@n7.RGB) {attributeSetTypes [#1]}
   Roughness=@n8.B
   Metallic=@n8.B
   Specular=@n8.B
  ) {attributeSetTypes [#2, #3, #4]}
  n10 = NamedRerouteUsage()
  n11 = TextureSampleParameter2D() {parameterName=S, samplerType=SAMPLERTYPE_Masks, texture=/Hera/Terrain/Rocks/Textures/Addon/T_Hera_SeaMist_RockClutter_A_A_S}
  n12 = SetMaterialAttributes!(
   MaterialAttributes=SetMaterialAttributes!(
    MaterialAttributes
    Normal=TextureSampleParameter2D() {parameterName=N, samplerType=SAMPLERTYPE_Normal, texture=/Hera/Terrain/Rocks/Textures/Addon/T_Hera_SeaMist_RockClutter_A_A_N}.RGB
   ) {attributeSetTypes [#1]}
   Roughness=@n11.B
   Metallic=@n11.B
   Specular=@n11.B
  ) {attributeSetTypes [#2, #3, #4]}
  n13 = NamedRerouteUsage()
  n14 = GetMaterialAttributes(None=StaticSwitchParameter(
    True=SetMaterialAttributes!(MaterialAttributes=@n9, Base Color=@n10) {attributeSetTypes [#5]}
    False=SetMaterialAttributes(MaterialAttributes=@n12, Base Color=@n13) {attributeSetTypes [#5]}
   ) {defaultValue=true, parameterName=IsNanite?}) {attributeGetTypes [#1]}
  n15 = NamedRerouteDeclaration(None=@n14.Normal) {name=Normal, nodeColor=<rgba 1 0.8906 0 1>}
  n16 = Constant() {r=0.5}
  n17 = GetMaterialAttributes(None=SetMaterialAttributes!(
    MaterialAttributes=@n14.MaterialAttributes
    Normal=MaterialFunctionCall(Normal=@n15, Flatness=@n16) {materialFunction=&FlattenNormal}.Result
   ) {attributeSetTypes [#1]}) {attributeGetTypes [#2]}
  n18 = LinearInterpolate(A=@n17.Roughness, Alpha=NamedRerouteUsage()) {constB=0.1}
  n19 = NamedRerouteUsage()
  MaterialFunctionCall(
   Material_In=SetMaterialAttributes!(MaterialAttributes=@n17.MaterialAttributes, World Position Offset, Roughness=LinearInterpolate(A=@n17.Roughness, B=@n18, Alpha=@n19)) {attributeSetTypes [#6, #2]}
   UseAdditiveWPO=StaticBool() {value=true}
  ) {materialFunction=&MF_FN_DestructibleDamage}
  n20 = Add(A=ComponentMask(None=TextureCoordinate() {coordinateIndex=1}) {r=true}) {constB=-0.5}
  n21 = Abs(None=@n20)
  n22 = OneMinus(None=Saturate(None=Round(None=@n21)))
  n23 = ScalarParameter() {group=Projector, parameterName=L1, sortPriority=-4}
  n24 = Abs(None=Add(A=@n20) {constB=-1})
  n25 = OneMinus(None=Saturate(None=Round(None=@n24)))
  n26 = ScalarParameter() {defaultValue=1, group=Projector, parameterName=L2, sortPriority=-3}
  NamedRerouteDeclaration(None=Max(A=Multiply(A=@n22, B=@n23), B=Multiply(A=@n25, B=@n26))) {name=Mask Layer, nodeColor=<rgba 1 0 0.5859 1>}
  NamedRerouteDeclaration(None=LinearInterpolate(
    A=TextureSampleParameter2D() {parameterName=Diffuse_Eye_Open_NITE, texture=/Hera/Terrain/Rocks/Textures/Addon/T_Hera_SeaMist_RockClutter_A_A_D_NITE}.RGB
    B=TextureSampleParameter2D() {parameterName=Diffuse_Eye_Closed_NITE, texture=/Hera/Terrain/Rocks/Textures/Addon/T_Hera_SeaMist_RockClutter_B_A_D_NITE}.RGB
    Alpha=NamedRerouteUsage()
   )) {name=Base Color Nanite, nodeColor=<rgba 0 0.8125 1 1>}
  n27 = Constant3Vector() {constant=<rgba 3.5412 2.3498 5.234 0>}
  n28 = ActorPositionWS()
  NamedRerouteDeclaration(None=Add(A=Time(), B=DotProduct(A=@n27, B=@n28))) {name=Random Time, nodeColor=<rgba 1 0 0.8203 1>}
  n29 = NamedRerouteUsage()
  n30 = Sine(None=Divide(A=@n29) {constB=3})
  n31 = Constant() {r=0.45}
  n32 = Cosine(None=Divide(A=@n29) {constB=4.3})
  n33 = Constant() {r=0.62}
  n34 = Divide(A=@n29) {constB=2.234}
  n35 = Sine(None=Divide(A=@n29) {constB=3.438})
  n36 = Multiply(A=Add(A=Fmod(A=@n30, B=@n31), B=Fmod(A=@n32, B=@n33)), B=LinearInterpolate(B=Sine(None=@n34), Alpha=Saturate(None=@n35)))
  NamedRerouteDeclaration(None=Saturate(None=Step(X=@n36) {constX=0.2, constY=0.175})) {name=Blink Mask, nodeColor=<rgba 0 1 0.1094 1>}
  NamedRerouteDeclaration(None=LinearInterpolate(
    A=TextureSampleParameter2D() {parameterName=Diffuse_Eye_Open, texture=/Hera/Terrain/Rocks/Textures/Addon/T_Hera_SeaMist_RockClutter_A_A_D}.RGB
    B=TextureSampleParameter2D() {parameterName=Diffuse_Eye_Closed, texture=/Hera/Terrain/Rocks/Textures/Addon/T_Hera_SeaMist_RockClutter_B_A_D}.RGB
    Alpha=NamedRerouteUsage()
   )) {name=Base Color Fallback, nodeColor=<rgba 0 0.8125 1 1>}
  n37 = StaticSwitchParameter(
   True=TextureSample() {samplerType=SAMPLERTYPE_Masks, texture=/Hera/Props/SeaMist/SeaMist_Starfish_A_A/Textures/T_SM_HeraSeaMist_StarFish_A_A_NITE_MSK}.RGB
   False=TextureSample() {samplerType=SAMPLERTYPE_Masks, texture=/Hera/Props/SeaMist/SeaMist_Starfish_A_A/Textures/T_SM_HeraSeaMist_StarFish_A_A_MSK}.RGB
  ) {defaultValue=true, parameterName=IsNanite?}
  NamedRerouteDeclaration(None=ComponentMask(None=@n37) {g=true, r=true}) {name=faceFeatureMaks, nodeColor=<rgba 0.8516 0 1 1>}
  NamedRerouteDeclaration(None=ComponentMask(None=@n37) {b=true}) {name=Eyes, nodeColor=<rgba 0 0.8594 1 1>}
  n38 = Multiply(A=TextureCoordinate(), B=Constant2Vector() {g=1, r=5})
  n39 = Constant2Vector() {g=0.1, r=-4.2}
  n40 = Multiply(A=Add(A=Add(A=@n38, B=@n39)) {constB=-0.5}) {constB=5}
  n41 = Saturate(None=Add(A=Abs(None=@n40)) {constB=-1})
  n42 = Multiply(A=@n41, B=@n41)
  OneMinus(None=Add(A=ComponentMask(None=@n42) {r=true}, B=ComponentMask(None=@n42) {g=true}))
  n43 = Transform(None=Multiply(A=LocalPosition().XYZ, B=Constant3Vector() {constant=<rgba 0 1 0 0>})) {transformSourceType=TRANSFORMSOURCE_Local}
  n44 = Constant() {r=3}
  n45 = ScalarParameter() {bUseCustomPrimitiveData=true, parameterName=BreathIntensity}
  n46 = Add(A=Abs(None=Add(A=@n44, B=@n45))) {constB=0.01}
  n47 = NamedRerouteUsage()
  n48 = ComponentMask(None=@n47) {g=true}
  n49 = Multiply(A=@n48, B=@n48)
  n50 = MaterialFunctionCall!(Sine Phase=Divide(A=Time()) {constB=5.5}, Value 1=Constant(), Value 2=LinearInterpolate(Alpha=Saturate(None=@n49)) {constA=2, constB=0.85}) {materialFunction=/Engine/Functions/Engine_MaterialFunctions02/Utility/Sine_Remapped}
  n51 = ComponentMask(None=@n47) {r=true}
  QualitySwitch(Default=Multiply(A=Divide(A=@n43, B=@n46), B=Multiply(A=@n50.Remapped Sine, B=@n51)), Low=Constant())
  TextureSample() {texture=/Hera/Terrain/Rocks/Textures/Addon/T_Hera_SeaMist_RockClutter_B_S_S}
 ]
}
